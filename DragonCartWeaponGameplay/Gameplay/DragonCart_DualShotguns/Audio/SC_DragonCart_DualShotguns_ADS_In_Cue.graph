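# Sound cue for the DragonCart dual shotguns' aim-down-sights-in foley sound.
# A random node picks between two ADS-in waves with equal weight, then a modulator node passes the result to the output.
# It plays in the weapon foley sound class at reduced volume, using the weapon reload attenuation settings.
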SoundCue /DragonCartWeaponGameplay/Gameplay/DragonCart_DualShotguns/Audio/SC_DragonCart_DualShotguns_ADS_In_Cue
root=Modulator_0
nodes [
 {id=Modulator_0, children [Random_0]}
 {id=Random_0, children [WavePlayer_0,WavePlayer_1], properties {weights [1,1]}}
 {
  id=WavePlayer_0
  properties {soundWaveAssetPtr=/DragonCartWeaponGameplay/Gameplay/DragonCart_DualShotguns/Audio/Waves/DragonCart_DualShotguns_ADS_In_02}
 }
 {
  id=WavePlayer_1
  properties {soundWaveAssetPtr=/DragonCartWeaponGameplay/Gameplay/DragonCart_DualShotguns/Audio/Waves/DragonCart_DualShotguns_ADS_In_01}
 }
]
properties {
 attenuationSettings=&Fort_Weapon_Reload_Att
 maxDistance=2250
 soundClassObject=&Fort_WeaponFoley
 volumeMultiplier=0.45
}
