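# Sound cue for the equip sound of the DragonCart Dual Shotguns weapon.
# A Random node picks one of three equally weighted equip waves and feeds a Modulator at the root.
# It plays as weapon foley using the shared Fort weapon reload attenuation.
# Concurrency overrides limit overlapping plays per owner, stopping the oldest.
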SoundCue /DragonCartWeaponGameplay/Gameplay/DragonCart_DualShotguns/Audio/SC_DragonCart_DualShotguns_Equip_Cue
root=Modulator_0
nodes [
 {id=Modulator_0, children [Random_1]}
 {id=Random_1, children [WavePlayer_3,WavePlayer_4,WavePlayer_5], properties {weights [1,1,1]}}
 {
  id=WavePlayer_3
  properties {soundWaveAssetPtr=/DragonCartWeaponGameplay/Gameplay/DragonCart_DualShotguns/Audio/Waves/DragonCart_DualShotguns_Equip_01}
 }
 {
  id=WavePlayer_4
  properties {soundWaveAssetPtr=/DragonCartWeaponGameplay/Gameplay/DragonCart_DualShotguns/Audio/Waves/DragonCart_DualShotguns_Equip_02}
 }
 {
  id=WavePlayer_5
  properties {soundWaveAssetPtr=/DragonCartWeaponGameplay/Gameplay/DragonCart_DualShotguns/Audio/Waves/DragonCart_DualShotguns_Equip_03}
 }
]
properties {
 attenuationSettings=&Fort_Weapon_Reload_Att
 maxDistance=2250
 soundClassObject=&Fort_WeaponFoley
 volumeMultiplier=1
 concurrencyOverrides {bLimitToOwner=true, maxCount=2, resolutionRule=StopOldest}
}
