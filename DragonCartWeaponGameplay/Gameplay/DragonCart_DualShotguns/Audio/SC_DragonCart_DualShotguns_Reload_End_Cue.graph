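# Sound cue for the end of the DragonCart dual shotguns' reload, classed as weapon foley.
# A Random node picks one of three equally weighted wave variations and feeds a Modulator that holds pitch fixed.
# The cue plays at reduced volume through the shared weapon reload attenuation.
# Concurrency overrides allow only a couple of instances per owner, stopping the oldest first.
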SoundCue /DragonCartWeaponGameplay/Gameplay/DragonCart_DualShotguns/Audio/SC_DragonCart_DualShotguns_Reload_End_Cue
root=Modulator_0
nodes [
 {id=Modulator_0, children [Random_0], properties {pitchMax=1, pitchMin=1}}
 {id=Random_0, children [WavePlayer_3,WavePlayer_4,WavePlayer_5], properties {weights [1,1,1]}}
 {
  id=WavePlayer_3
  properties {soundWaveAssetPtr=/DragonCartWeaponGameplay/Gameplay/DragonCart_DualShotguns/Audio/Waves/DragonCart_DualShotguns_Reload_End_01}
 }
 {
  id=WavePlayer_4
  properties {soundWaveAssetPtr=/DragonCartWeaponGameplay/Gameplay/DragonCart_DualShotguns/Audio/Waves/DragonCart_DualShotguns_Reload_End_02}
 }
 {
  id=WavePlayer_5
  properties {soundWaveAssetPtr=/DragonCartWeaponGameplay/Gameplay/DragonCart_DualShotguns/Audio/Waves/DragonCart_DualShotguns_Reload_End_03}
 }
]
properties {
 attenuationSettings=&Fort_Weapon_Reload_Att
 maxDistance=2250
 soundClassObject=&Fort_WeaponFoley
 volumeMultiplier=0.5
 concurrencyOverrides {bLimitToOwner=true, maxCount=2, resolutionRule=StopOldest}
}
