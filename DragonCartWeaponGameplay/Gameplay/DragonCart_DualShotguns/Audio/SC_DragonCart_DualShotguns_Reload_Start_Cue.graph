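# Sound cue for the start of the DragonCart dual shotguns reload.
# A random node picks one of three equally weighted wave variations and feeds a modulator that applies a fixed upward pitch shift.
# It uses the shared weapon reload attenuation and weapon foley sound class, with owner-limited concurrency that stops the oldest instance when the cap is exceeded.
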SoundCue /DragonCartWeaponGameplay/Gameplay/DragonCart_DualShotguns/Audio/SC_DragonCart_DualShotguns_Reload_Start_Cue
root=Modulator_0
nodes [
 {id=Modulator_0, children [Random_0], properties {pitchMax=1.1, pitchMin=1.1}}
 {id=Random_0, children [WavePlayer_0,WavePlayer_1,WavePlayer_2], properties {weights [1,1,1]}}
 {
  id=WavePlayer_0
  properties {soundWaveAssetPtr=/DragonCartWeaponGameplay/Gameplay/DragonCart_DualShotguns/Audio/Waves/DragonCart_DualShotguns_Reload_Start_01}
 }
 {
  id=WavePlayer_1
  properties {soundWaveAssetPtr=/DragonCartWeaponGameplay/Gameplay/DragonCart_DualShotguns/Audio/Waves/DragonCart_DualShotguns_Reload_Start_02}
 }
 {
  id=WavePlayer_2
  properties {soundWaveAssetPtr=/DragonCartWeaponGameplay/Gameplay/DragonCart_DualShotguns/Audio/Waves/DragonCart_DualShotguns_Reload_Start_03}
 }
]
properties {
 attenuationSettings=&Fort_Weapon_Reload_Att
 maxDistance=2250
 soundClassObject=&Fort_WeaponFoley
 volumeMultiplier=1
 concurrencyOverrides {bLimitToOwner=true, maxCount=2, resolutionRule=StopOldest}
}
