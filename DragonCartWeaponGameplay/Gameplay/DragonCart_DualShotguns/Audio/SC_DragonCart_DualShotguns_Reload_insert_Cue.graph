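# Sound cue for the shell insert sound during the DragonCart Dual Shotguns reload, classed as weapon foley.
# A Random node picks among three equally weighted insert wave variations and feeds a Modulator at the root for slight variation.
# It uses shared weapon reload attenuation at reduced volume, with per-owner concurrency that stops the oldest instance when a small voice limit is exceeded.
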SoundCue /DragonCartWeaponGameplay/Gameplay/DragonCart_DualShotguns/Audio/SC_DragonCart_DualShotguns_Reload_insert_Cue
root=Modulator_0
nodes [
 {id=Modulator_0, children [Random_2]}
 {id=Random_2, children [WavePlayer_6,WavePlayer_7,WavePlayer_8], properties {weights [1,1,1]}}
 {
  id=WavePlayer_6
  properties {soundWaveAssetPtr=/DragonCartWeaponGameplay/Gameplay/DragonCart_DualShotguns/Audio/Waves/DragonCart_DualShotguns_Reload_Insert_01}
 }
 {
  id=WavePlayer_7
  properties {soundWaveAssetPtr=/DragonCartWeaponGameplay/Gameplay/DragonCart_DualShotguns/Audio/Waves/DragonCart_DualShotguns_Reload_Insert_02}
 }
 {
  id=WavePlayer_8
  properties {soundWaveAssetPtr=/DragonCartWeaponGameplay/Gameplay/DragonCart_DualShotguns/Audio/Waves/DragonCart_DualShotguns_Reload_Insert_03}
 }
]
properties {
 attenuationSettings=&Fort_Weapon_Reload_Att
 maxDistance=2250
 soundClassObject=&Fort_WeaponFoley
 volumeMultiplier=0.5
 concurrencyOverrides {bLimitToOwner=true, maxCount=2, resolutionRule=StopOldest}
}
