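; Dual automatic shotgun weapon for DragonCart that adjusts the parent weapon's muzzle and shell effects.
; For bot-controlled instigators it skips parent hand-fired handling, fires both hands' flash and shell effects together, and deactivates them directly on stop.
; Shell and muzzle-flash overrides add per-hand shell bursts and extra smoke after calling the parent.
; As a marked hack, BeginPlay waits briefly, then pushes shell settings into both burst Niagara components.
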(event EventHandFired
  (params (Hand :type EDualWeaponHand%20Enum :default "LEFT") (bPersistantFire :type Boolean))
  ;[1 1 0 0.5 32] Bot only-- prevent effects due to problems with the default setup.
  (if (not (Pawn|IsBotControlled (Variables|Actor|GetInstigator)))
    (|CallParentFunction Hand bPersistantFire)))

(event EventOnPlayWeaponFireFX
  (params (bPersistentFire :type Boolean) (bSecondaryFire :type Boolean))
  ;[1 1 0 0.5 32] Bot only-- override effects due to problems with the default setup.
  (if (Pawn|IsBotControlled (Variables|Actor|GetInstigator))
    (CallFunction|MuzzleFlashFX false)
    (CallFunction|MuzzleFlashFX false true)
    (CallFunction|ActivateShellsFX (Variables|Effects%28Shells%29|GetUseShellsOnFire?))
    (CallFunction|ActivateShellsFX (Variables|Effects%28Shells%29|GetUseShellsOnFire?) true)))

(event EventOnStopWeaponFireFX
  ;[1 1 0 0.5 32] Bot only-- prevent effects due to problems with the default setup.
  (if (Pawn|IsBotControlled (Variables|Actor|GetInstigator))
    (CallFunction|DeactivateMuzzleFX)
    (CallFunction|DeactivateShellsFX)
    (else
      (|CallParentFunction))))

(event EventBeginPlay
  ;[1 0 0 0.5] Hack - initial shell setup. -Yue
  (|CallParentFunction)
  (Utilities|FlowControl|Delay 0.1)
  (for _array_element (Utilities|Array|MakeArray (Variables|Default|GetNS_ShellsFire_Sound_Burst_R) (Variables|Default|GetNS_ShellsFire_Sound_Burst_L))
    (Niagara|SetNiagaraVariable(Vector3) _array_element "ShellsRotationRate" (Variables|Effects%28Shells%29|GetShellsRotationRate))
    (Niagara|SetNiagaraVariable(Vector3) _array_element "Velocity" (Variables|Effects%28Shells%29|GetShells%20Velocity))
    (Niagara|SetNiagaraVariable(Vector3) _array_element "Gravity" (Variables|Effects%28Shells%29|GetShells%20Gravity))
    (Niagara|SetNiagaraVariable(Vector3) _array_element "Size" (Variables|Effects%28Shells%29|GetShells%20Size))
    (Niagara|SetNiagaraVariable(Vector3) _array_element "InheritParentVelocity" "1.000000,1.000000,1.000000")
    (Niagara|SetNiagaraVariable(Int32) _array_element "Shell Type" 4)))

(event OnComponentActivated(LeftMuzzle(Empty)Niagara)
  (params (Component :type Actor%20Component%20Object%20Reference) (bReset :type Boolean)))

(event OnComponentActivated(Muzzle(Empty)Niagara)
  (params (Component :type Actor%20Component%20Object%20Reference) (bReset :type Boolean)))

(fn ActivateShellsFX
  (params (Bool :type Boolean) (OffHand :type Boolean))
  (|CallParentFunction true true)
  ;[1 0 0 0.5] Shell Spawn logic hack. -Yue
  (if OffHand
    (Components|Activation|Activate (Variables|Default|GetNS_ShellsFire_Sound_Burst_R) true)
    (else
      (Components|Activation|Activate (Variables|Default|GetNS_ShellsFire_Sound_Burst_L) true))))

(fn MuzzleFlashFX
  (params (Persistent%20Fire :type Boolean) (OffHand :type Boolean))
  (|CallParentFunction Persistent%20Fire OffHand)
  (if OffHand
    (Components|Activation|Activate (Variables|Default|GetAdditionalSmoke) true)
    (else
      (Components|Activation|Activate (Variables|Default|GetAdditionalSmoke1) true))))

(fn ConstructionScript
  (|CallParentFunction))
